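# Physics asset for the face accessory of the GroovyReader WhiteCap medium female player cosmetic, giving its wire bones capsule bodies and joint limits.
# Center, left and right wire bodies each constrain directly to base_wires rather than chaining, mostly with one narrow swing axis and minimal twist.
# The back wire forms a two-link chain from base_wires, with a looser twist at its tip.
# Larger spine, neck and clavicle capsules approximate the upper body and carry no constraints.
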PhysicsAsset /BRCosmetics/Characters/Player/Female/Medium/Bodies/F_MED_GroovyReader_WhiteCap/Meshes/Parts/F_MED_GroovyReader_WhiteCap_FaceAcc_Physics
bodies {
 base_wires [{shapeType=Capsule, length=0.505, radius=0.505, shapeName=base_wires_capsule}]
 clavicle_l [{shapeType=Capsule, length=1.25, radius=6.9, shapeName=spine_05_capsule, center=<4.5 0 -0.75>, rotation=<rot 90 0 0>}]
 clavicle_r [{shapeType=Capsule, length=1.25, radius=6.9, shapeName=spine_05_capsule, center=<-4.5 0 0.75>, rotation=<rot 90 0 0>}]
 dyn_wires_back [{shapeType=Capsule, length=4, radius=1, shapeName=dyn_wires_back_capsule, center=<2.5 0 0>, rotation=<rot 90 0 0>}]
 dyn_wires_back_01 [{shapeType=Capsule, length=2, radius=2.1, shapeName=dyn_wires_back_capsule, center=<2.75 0 0>, rotation=<rot 84.8467 -180 180>}]
 dyn_wires_c_01 [{shapeType=Capsule, length=3, radius=0.5, shapeName=dyn_wires_c_01_capsule, center=<2 0 0>, rotation=<rot 90 0 0>}]
 dyn_wires_c_02 [{shapeType=Capsule, length=2, radius=0.5, shapeName=dyn_wires_c_02_capsule, center=<1.5 0 0>, rotation=<rot 90 0 0>}]
 dyn_wires_c_03 [{shapeType=Capsule, length=3, radius=0.5, shapeName=dyn_wires_c_03_capsule, center=<2 0 0>, rotation=<rot 90 0 0>}]
 dyn_wires_l_01 [{shapeType=Capsule, length=2, radius=0.5, shapeName=dyn_wires_l_01_capsule, center=<1.5 0 0>, rotation=<rot 90 0 0>}]
 dyn_wires_l_02 [{shapeType=Capsule, length=2, radius=0.5, shapeName=dyn_wires_l_02_capsule, center=<1.5 0 0>, rotation=<rot 90 0 0>}]
 dyn_wires_l_03 [{shapeType=Capsule, length=3, radius=0.5, shapeName=dyn_wires_l_03_capsule, center=<3 0 0>, rotation=<rot 90 0 0>}]
 dyn_wires_l_04 [{shapeType=Capsule, length=3, radius=0.5, shapeName=dyn_wires_l_04_capsule, center=<2 0 0>, rotation=<rot 90 0 0>}]
 dyn_wires_l_05 [{shapeType=Capsule, length=3, radius=0.5, shapeName=dyn_wires_l_05_capsule, center=<2 0 0>, rotation=<rot 90 0 0>}]
 dyn_wires_r_01 [{shapeType=Capsule, length=2, radius=0.5, shapeName=dyn_wires_r_01_capsule, center=<1.5 0 0>, rotation=<rot 90 0 0>}]
 dyn_wires_r_02 [{shapeType=Capsule, length=2, radius=0.5, shapeName=dyn_wires_r_02_capsule, center=<1.5 0 0>, rotation=<rot 90 0 0>}]
 dyn_wires_r_03 [{shapeType=Capsule, length=3, radius=0.5, shapeName=dyn_wires_r_03_capsule, center=<2 0 0>, rotation=<rot 90 0 0>}]
 dyn_wires_r_04 [{shapeType=Capsule, length=4, radius=0.5, shapeName=dyn_wires_r_04_capsule, center=<2.5 0 0>, rotation=<rot 90 0 0>}]
 dyn_wires_r_05 [{shapeType=Capsule, length=3, radius=0.5, shapeName=dyn_wires_r_04_capsule, center=<2 0 0>, rotation=<rot 90 0 0>}]
 neck_01 [{shapeType=Capsule, length=15, radius=3.75, shapeName=spine_05_capsule, center=<3 0 0>}]
 spine_05 [{shapeType=Capsule, length=4, radius=8.35, shapeName=spine_05_capsule, center=<3.15 0 0>}]
}
constraints [
 {bone1Name=dyn_wires_c_01, bone2Name=base_wires, swing1LimitDegrees=2, swing2LimitDegrees=30, twistLimitDegrees=3}
 {bone1Name=dyn_wires_c_02, bone2Name=base_wires, swing1LimitDegrees=60, swing2LimitDegrees=5, twistLimitDegrees=2}
 {bone1Name=dyn_wires_c_03, bone2Name=base_wires, swing1LimitDegrees=45, swing2LimitDegrees=5, twistLimitDegrees=1}
 {bone1Name=dyn_wires_l_01, bone2Name=base_wires, swing1LimitDegrees=5, swing2LimitDegrees=30, twistLimitDegrees=2}
 {bone1Name=dyn_wires_l_02, bone2Name=base_wires, swing1LimitDegrees=5, swing2LimitDegrees=45, twistLimitDegrees=3}
 {bone1Name=dyn_wires_l_03, bone2Name=base_wires, swing1LimitDegrees=5, swing2LimitDegrees=20, twistLimitDegrees=1}
 {bone1Name=dyn_wires_l_04, bone2Name=base_wires, swing1LimitDegrees=5, swing2LimitDegrees=35, twistLimitDegrees=2}
 {bone1Name=dyn_wires_l_05, bone2Name=base_wires, swing1LimitDegrees=5, swing2LimitDegrees=60, twistLimitDegrees=3}
 {bone1Name=dyn_wires_r_01, bone2Name=base_wires, swing1LimitDegrees=5, swing2LimitDegrees=30, twistLimitDegrees=3}
 {bone1Name=dyn_wires_r_02, bone2Name=base_wires, swing1LimitDegrees=5, swing2LimitDegrees=45, twistLimitDegrees=2}
 {bone1Name=dyn_wires_r_03, bone2Name=base_wires, swing1LimitDegrees=5, swing2LimitDegrees=40, twistLimitDegrees=3}
 {bone1Name=dyn_wires_r_04, bone2Name=base_wires, swing1LimitDegrees=5, swing2LimitDegrees=20, twistLimitDegrees=2}
 {bone1Name=dyn_wires_back_01, bone2Name=dyn_wires_back, swing1LimitDegrees=20, swing2LimitDegrees=25, twistLimitDegrees=40}
 {bone1Name=dyn_wires_back, bone2Name=base_wires, swing1LimitDegrees=45, swing2LimitDegrees=60, twistLimitDegrees=5}
 {bone1Name=dyn_wires_r_05, bone2Name=base_wires, swing1LimitDegrees=5, swing2LimitDegrees=19, twistLimitDegrees=2}
]
